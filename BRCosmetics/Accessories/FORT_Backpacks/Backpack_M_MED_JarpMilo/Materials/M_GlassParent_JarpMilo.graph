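# Translucent, unlit parent glass material for the JarpMilo backpack cosmetic, built on the 2023 character parent functions.
# Emissive fakes lighting from a height-based color gradient, fresnel-banded skylight reflections, time-of-day scaling and fixed-axis Blinn-Phong specular, blended with base color by Tech Art Mask green.
# Scene-depth thickness and normal-map bumps shape an opacity term, while mask red modulates final opacity.
# Sun-offset sphere masks form a fake shadow, with a VisShadow switch.
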Material /BRCosmetics/Accessories/FORT_Backpacks/Backpack_M_MED_JarpMilo/Materials/M_GlassParent_JarpMilo
graph {
 properties {
  bEnableResponsiveAA=true
  bScreenSpaceReflections=true
  bUseMaterialAttributes=true
  bUsedWithMeshDeformer=true
  bUsedWithMorphTargets=true
  bUsedWithSkeletalMesh=true
  blendMode=BLEND_Translucent
  ditherOpacityMask=true
  pixelDepthOffsetMode=PDOM_Legacy
  shadingModel=MSM_Unlit
  translucencyDirectionalLightingIntensity=5
  translucencyLightingMode=TLM_SurfacePerPixelLighting
  usedShadingModels=Unlit
 }
 nodes [
  Constant() {r=1}
  GetMaterialAttributes() {attributeGetTypes [#1, #2]}
  MaterialFunctionCall() {materialFunction=/Game/Global/Materials/Character_Parent_2023/Anime_Parent_2023/MaterialFunctions/MFC_AnimeShader_LightDirection}
  n1 = MaterialFunctionCall() {materialFunction=&MF_CharacterParent_2023_PreFX}
  n2 = GetMaterialAttributes(None=@n1.Result) {attributeGetTypes [#3, #1]}
  n3 = NamedRerouteUsage()
  n4 = ComponentMask(None=NamedRerouteUsage()) {b=true}
  n5 = ScalarParameter() {defaultValue=-50, group=DepthColor, parameterName=DepthBottom}
  n6 = Saturate(None=Divide(A=Subtract(A=@n4, B=@n5), B=ScalarParameter() {defaultValue=70, group=DepthColor, parameterName=DepthGradient}))
  n7 = ScalarParameter() {defaultValue=2, group=DepthColor, parameterName=DepthPow}
  n8 = VectorParameter() {parameterName=BaseColor, defaultValue=<rgba 1 1 1 1>}
  n9 = PixelNormalWS()
  n10 = Multiply(
   A=SkyLightEnvMapSample(Direction=ReflectionVectorWS(CustomWorldNormal=@n9), Roughness=ScalarParameter() {parameterName=Roughness})
   B=ScalarParameter() {defaultValue=1, parameterName=ReflectionStrength}
  )
  n11 = ScalarParameter() {defaultValue=0.7, parameterName=OuterEdgeThickness}
  n12 = Subtract(A=@n11, B=ScalarParameter() {defaultValue=0.05, parameterName=OuterEdgeFade})
  n13 = Fresnel() {baseReflectFraction=0, exponent=1}
  n14 = OneMinus(None=SmoothStep(Min=@n12, Max=@n11, Value=@n13))
  n15 = ScalarParameter() {parameterName=InnerEdgeFadeBias}
  n16 = SmoothStep(Min=Constant(), Max=Saturate(None=Subtract(A=@n12, B=@n15)), Value=@n13)
  n17 = Saturate(None=Power(Base=Multiply(A=@n14, B=@n16), Exp=ScalarParameter() {defaultValue=2, parameterName=Exp}))
  n18 = Add(
   A=LinearInterpolate(
    A=VectorParameter() {parameterName=DepthColor, defaultValue=<rgba 0 0.7593 1 1>}.RGB
    B=VectorParameter() {parameterName=DepthColorLighter, defaultValue=<rgba 0 0.7593 1 1>}.RGB
    Alpha=Power(Base=@n6, Exp=@n7)
   )
   B=Multiply(A=Multiply(A=@n8.RGB, B=@n10), B=@n17)
  )
  n19 = Constant() {r=0.5}
  n20 = MaterialFunctionCall(Value_Morning=@n19, Value_Day=Constant() {r=1}, Value_Evening=@n19, Value_Night=Constant() {r=0.3}) {materialFunction=&Scalar_Time_of_Day}
  n21 = Subtract(
   A=SceneDepth().R
   B=MaterialFunctionCall!(World Position) {materialFunction=/Engine/Functions/Engine_MaterialFunctions02/Utility/DepthFromWorldPosition}.Result
  )
  n22 = ScalarParameter() {defaultValue=100, group=DepthColor, parameterName=Thickness}
  n23 = Multiply(
   A=Power(Base=Divide(A=@n21, B=@n22), Exp=ScalarParameter() {defaultValue=2, group=DepthColor, parameterName=DepthPower})
   B=ScalarParameter() {defaultValue=1, group=DepthColor, parameterName=DepthAmount}
  )
  n24 = MaterialFunctionCall!(Material=SetMaterialAttributes!(MaterialAttributes, Emissive Color=Multiply(A=@n18, B=@n20.Result), Opacity=Add(A=@n17, B=@n23)) {attributeSetTypes [#1, #2]}) {materialFunction=&/MF_FakeBlinnPhongSpec_FixedAxis}
  n25 = LinearInterpolate(
   A=Add(A=@n2.Emissive Color, B=GetMaterialAttributes(None=@n24.Result) {attributeGetTypes [#1]}.Emissive Color)
   B=@n2.Base Color
   Alpha=ComponentMask(None=@n1.Tech Art Mask) {g=true}
  )
  n26 = NamedRerouteUsage()
  n27 = Power(
   Base=ComponentMask(None=@n1.Tech Art Mask) {r=true}
   Exp=ScalarParameter() {defaultValue=3, group=DepthColor, parameterName=OpacityMaskPower}
  )
  MaterialFunctionCall(In=SetMaterialAttributes(
    MaterialAttributes=@n2.MaterialAttributes
    Emissive Color=StaticSwitchParameter(True=@n3, False=@n25) {parameterName=VisShadow}
    Opacity=LinearInterpolate(A=@n26, Alpha=@n27)
   ) {attributeSetTypes [#1, #2]}) {materialFunction=&MF_TechArtOverrides}
  Multiply(A=@n9) {constB=-1}
  NamedRerouteDeclaration(None=@n17) {name=FresnelOpacity, nodeColor=<rgba 0.8516 0 1 1>}
  NamedRerouteDeclaration(None=@n1.Pre skinned Local Position) {name=PreskinnedPos, nodeColor=<rgba 0.2422 0 1 1>}
  n28 = TextureSample() {samplerType=SAMPLERTYPE_Normal, texture=/BRCosmetics/Accessories/FORT_Backpacks/Backpack_M_MED_JarpMilo/Textures/T_JarpMilo_Backpack_N}
  n29 = Constant3Vector() {constant=<rgba 0 0 1 0>}
  n30 = Saturate(None=Multiply(A=Distance(A=@n28.RGB, B=@n29), B=ScalarParameter() {defaultValue=1, group=SurfaceColor, parameterName=NormalBumpUP}) {constB=10})
  n31 = ScalarParameter() {defaultValue=0.8, group=DepthColor, parameterName=MaxDepthOpacity}
  n32 = Saturate(None=@n23)
  NamedRerouteDeclaration(None=Multiply(A=OneMinus(None=@n30), B=LinearInterpolate(B=@n31, Alpha=@n32))) {name=OpacityOUT, nodeColor=<rgba 0.9453 1 0 1>}
  n33 = NamedRerouteUsage()
  n34 = VectorParameter() {parameterName=Center, defaultValue=<rgba 0 0 0 1>}
  n35 = SphereMask(
   A=@n33
   B=@n34.RGB
   Radius=ScalarParameter() {defaultValue=200, group=FakeLighting, parameterName=CenterRadius}
   Hardness=ScalarParameter() {defaultValue=0.5, group=FakeLighting, parameterName=CenterHardness}
  ) {hardnessPercent=0}
  n36 = Multiply(A=Transform(None=AtmosphericLightVector()) {transformSourceType=TRANSFORMSOURCE_World, transformType=TRANSFORM_Local}) {constB=-1}
  n37 = ScalarParameter() {defaultValue=10, group=FakeLighting, parameterName=CenterOffset}
  n38 = SphereMask(
   A=Add(A=Multiply(A=@n36, B=@n37), B=@n34.RGB)
   B=@n33
   Radius=ScalarParameter() {defaultValue=20, group=FakeLighting, parameterName=CenterRadius02}
   Hardness=ScalarParameter() {defaultValue=0.5, group=FakeLighting, parameterName=CenterHardness02}
  ) {hardnessPercent=0}
  NamedRerouteDeclaration(None=Saturate(None=Add(A=@n35, B=@n38))) {name=FakeShadowMask, nodeColor=<rgba 1 0.1172 0 1>}
  Subtract(A=@n12, B=ScalarParameter() {defaultValue=0.65, parameterName=InnerEdgeThickness})
 ]
}
