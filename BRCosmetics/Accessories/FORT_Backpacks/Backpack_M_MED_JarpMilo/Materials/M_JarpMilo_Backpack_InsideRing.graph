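# Lit character-accessory material for the inside ring of the JarpMilo backpack, adding glowing, water-like caustics.
# Panning caustics, distorted by reflection and normal-map refraction, blend two colors, scale by time of day and add to emissive between CharacterParent PreFX and PostFX functions.
# A separate branch computes UV-masked, time-animated sine wobble along vertex normals from world and pre-skinned positions, with no visible link into that attribute chain.
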Material /BRCosmetics/Accessories/FORT_Backpacks/Backpack_M_MED_JarpMilo/Materials/M_JarpMilo_Backpack_InsideRing
graph {
 properties {
  bUseMaterialAttributes=true
  bUsedWithMeshDeformer=true
  bUsedWithSkeletalMesh=true
  ditherOpacityMask=true
  opacityMaskClipValue=0.5
  pixelDepthOffsetMode=PDOM_Legacy
  usedShadingModels=Default Lit
 }
 nodes [
  n1 = MaterialFunctionCall() {materialFunction=&MF_CharacterParent_2023_PreFX}
  n2 = GetMaterialAttributes(None=@n1.Result) {attributeGetTypes [#1]}
  n3 = VectorParameter() {group=ColorShading, parameterName=CausticHigh, defaultValue=<rgba 0.1635 0.9303 1 1.3>}
  n4 = ComponentMask(None=NamedRerouteUsage()) {b=true, r=true}
  n5 = ScalarParameter() {defaultValue=60, group=UV, parameterName=CuasticsUnit}
  n6 = AppendVector(
   A=ScalarParameter() {group=UV, parameterName=CausticsPannerX}
   B=ScalarParameter() {defaultValue=-0.1, group=UV, parameterName=CausticsPannerY}
  )
  n7 = ReflectionVectorWS()
  n8 = Constant3Vector() {constant=<rgba 0 0 1 0>}
  n9 = TextureSampleParameter2D() {
   parameterName=SurfaceN
   samplerType=SAMPLERTYPE_Normal
   texture=/BRCosmetics/Accessories/FORT_Backpacks/Backpack_M_MED_JarpMilo/Textures/T_JarpMilo_Backpack_N
  }
  n10 = Transform(None=@n9.RGB)
  n11 = Constant() {r=-1}
  n12 = Length(None=MaterialFunctionCall!(
    Ray Direction=Transform(None=@n8)
    Surface Normal=Multiply(A=@n10, B=@n11)
    Refractive Index Origin=Constant() {r=1}
    Refractive Index Target=Constant() {r=1.33}
   ) {materialFunction=/Engine/Functions/Engine_MaterialFunctions01/Vectors/Refract}.Vector) {is=class MaterialExpressionLength}
  n13 = ScalarParameter() {defaultValue=0.04, parameterName=DisortIntensity}
  n14 = Add(A=Panner(Coordinate=Divide(A=@n4, B=@n5), Speed=@n6), B=Multiply(A=ComponentMask(None=@n7) {b=true, r=true}, B=Multiply(A=@n12, B=@n13)))
  n15 = Constant() {r=0.5}
  n16 = Multiply(
   A=LinearInterpolate(
    A=VectorParameter() {group=ColorShading, parameterName=BackGroundColor, defaultValue=<rgba 0 0.3079 0.9237 1>}.RGB
    B=Multiply(A=@n3.RGB, B=@n3.RGB)
    Alpha=TextureSample(UVs=@n14) {texture=/Game/Global/Textures/Default/Caustics/T_Caustics_Softer}.R
   )
   B=MaterialFunctionCall(Value_Morning=@n15, Value_Day=Constant() {r=1}, Value_Evening=@n15, Value_Night=Constant() {r=0.2}) {materialFunction=&Scalar_Time_of_Day}.Result
  )
  MaterialFunctionCall(
   Material=SetMaterialAttributes(MaterialAttributes=@n2.MaterialAttributes, Emissive Color=Add(A=@n2.Emissive Color, B=@n16)) {attributeSetTypes [#1]}
   M=@n1.M
   Pre Skinned Local Position=@n1.M
   Distance Blend=@n1.M
   Fresnel=@n1.M
  ) {materialFunction=&MF_CharacterParent_2023_PostFX}
  MaterialFunctionCall() {materialFunction=&MF_TechArtOverrides_accessories}
  MaterialFunctionCall(Normal=@n9.RGB, Flatness=Constant() {r=-1}) {materialFunction=&FlattenNormal}
  n17 = VertexNormalWS()
  n18 = WorldPosition() {worldPositionShaderOffset=WPT_ExcludeAllShaderOffsets}
  n19 = Subtract(A=Frac(None=Multiply(A=@n18.XYZ) {constB=0.0025})) {constB=0.5}
  n20 = MaterialFunctionCall(Float3=Abs(None=Multiply(A=@n19) {constB=2})) {materialFunction=&BreakOutFloat3Components}
  n21 = Subtract(A=Add(A=Add(A=@n20.R, B=@n20.R), B=@n20.B)) {constB=2}
  n22 = Frac(None=Multiply(A=Time()) {constB=0.25})
  n23 = Add(A=Sine(None=Subtract(A=@n21, B=@n22)))
  n24 = MaterialFunctionCall(Float3=Multiply(A=PreSkinnedPosition()) {constB=0.02}) {materialFunction=&BreakOutFloat3Components}
  n25 = Add(A=Add(A=@n24.R, B=@n24.R), B=@n24.B)
  n26 = Multiply(A=Subtract(A=Time(), B=ScalarParameter() {parameterName=Slurp_WPO_random_seed})) {constB=0.6}
  n27 = Sine(None=Subtract(A=Subtract(A=@n25) {constB=2}, B=Frac(None=@n26)))
  n28 = Add(
   A=Multiply(A=Multiply(A=@n23) {constB=0.5}, B=ScalarParameter() {defaultValue=1, parameterName=WPO_WorldNoiseScale})
   B=Multiply(A=Add(A=@n27)) {constB=0.5}
  )
  n29 = Saturate(None=MaterialFunctionCall() {materialFunction=/Engine/Functions/Engine_MaterialFunctions02/Texturing/3DSandMayaUVCoordinates}.UVChannel1)
  Multiply(
   A=Multiply(A=Multiply(A=@n17, B=@n28), B=ComponentMask(None=@n29) {r=true})
   B=ScalarParameter() {defaultValue=1, parameterName=WPO_FinalScale}
  )
  NamedRerouteDeclaration(None=@n1.Pre skinned Local Position) {name=PreskinnedPOS, nodeColor=<rgba 0 1 0.3438 1>}
  Normalize(VectorInput=@n6)
  TextureCoordinate()
 ]
}
