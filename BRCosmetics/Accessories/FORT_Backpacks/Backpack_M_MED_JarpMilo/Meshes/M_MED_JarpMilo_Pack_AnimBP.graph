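# Animation Blueprint for the JarpMilo backpack cosmetic, playing its pack animation and offsetting the pack by player state.
# The AnimGraph feeds the sequence into a Transform Bone node that additively moves and rotates base_pack in bone space.
# Each animation update sets that translation and rotation, using a larger offset in the front end or while skydiving and a smaller one otherwise.
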AnimBlueprint /BRCosmetics/Accessories/FORT_Backpacks/Backpack_M_MED_JarpMilo/Meshes/M_MED_JarpMilo_Pack_AnimBP
target_skeleton=&/M_MED_JarpMilo_Pack_Skeleton
graphs [
 AnimGraph {nodes [
   Input Pose()
   n1 = "Transform (Modify) Bone - Bone: base_pack"(
    ComponentPose=Local To Component(LocalPose=Sequence Player 'Backpack_JarpMilo_pack'() {sequence=/BRCosmetics/Accessories/FORT_Backpacks/Backpack_M_MED_JarpMilo/Anim/Backpack_JarpMilo_pack})
    Alpha=1
    Translation=Get SkyTras() -> (SkyTras:Vector)
    Rotation=Get SkyRot() -> (SkyRot:Rotator)
    Scale="1,1,1"
   ) {
    rotationMode=BMM_Additive
    rotationSpace=BCS_BoneSpace
    translationMode=BMM_Additive
    translationSpace=BCS_BoneSpace
    alphaBoolBlend {blendOption=Linear}
    boneToModify {boneName=base_pack}
   }
   Output Pose(Result=Component To Local(ComponentPose=@n1)) {layerGroup=None}
  ]}
 EventGraph {nodes [
   Event BlueprintUpdateAnimation()
   > Sequence()
   > n1 = Branch(Condition=OR Boolean(A=Get bIsInFrontEnd(), B=Get bIsSkyDiving()))
   > Set SkyRot(SkyRot:Rotator="0,0,10") -> (Output_Get:Rotator)
   > Set SkyTras(SkyTras:Vector="0,-4,0") -> (Output_Get:Vector)
   Set SkyRot(execute=@n1.else, SkyRot="0,0,5")
   > Set SkyTras(SkyTras="0,-1.5,0")
   TryGetPawnOwner()
  ]}
]
variables [
 SkyRot {type=Rotator}
 SkyTras {type=Vector}
]
